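# Looping foley sound cue for the JazzShoes emote, layering a continuous loop with repeating chainsaw sounds.
# A Mixer combines the looping base wave with a Looping node that repeatedly triggers an equally weighted Random pick from three chainsaw variations.
# It uses emote attenuation and the foley sound class, keeps playing when silent, caps concurrency, and lists a music audio-analysis submix send.
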SoundCue /BRCosmetics/Sounds/Emotes/JazzShoes/EmoteFoley_JazzShoes_Loop_Cue
root=Mixer_2
nodes [
 {id=Looping_1, children [Random_0]}
 {id=Mixer_2, children [WavePlayer_0,Looping_1], properties {inputVolume [1,1]}}
 {id=Random_0, children [WavePlayer_6,WavePlayer_7,WavePlayer_8], properties {weights [1,1,1]}}
 {id=WavePlayer_0, properties {bLooping=true, soundWaveAssetPtr=&/IP_Emote_JazzShoes_Loop_01}}
 {id=WavePlayer_6, properties {soundWaveAssetPtr=&/IP_Emote_JazzShoes_Chainsaw_01}}
 {id=WavePlayer_7, properties {soundWaveAssetPtr=&/IP_Emote_JazzShoes_Chainsaw_02}}
 {id=WavePlayer_8, properties {soundWaveAssetPtr=&/IP_Emote_JazzShoes_Chainsaw_03}}
]
properties {
 attenuationSettings=&Fort_Emotes_Athena_Att
 audioPropertiesSheet=/BRCosmetics/Sounds/Settings/PropertySheets/Emotes/APS_Cosmetic_Emote_Foley
 bOverrideConcurrency=true
 maxDistance=1600
 soundClassObject=/Game/Sounds/FortSoundClassesAndModes/Fort_Foley_Emote
 virtualizationMode=PlayWhenSilent
 volumeMultiplier=1
 concurrencyOverrides {maxCount=3}
 soundSubmixSends [{minSendLevel=0, soundSubmix=&Fortnite_AudioAnalysis_Music}]
}
